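# Animation Blueprint for the OysterKnock player body, layering Control Rig motion over a pose copied from the attached parent mesh.
# The update event feeds the rig emote, skydive-floating and matching OysterKnock glider states as floats, alongside parachute state.
# A per-bone blend applies DefaultSlot to both shoulder pad branches, weighted by the front-end instance's Frontend_Idle_Bypass curve, read only in the front end.
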AnimBlueprint /BRCosmetics/Characters/Player/Male/Large/Bodies/M_LRG_OysterKnock/Meshes/M_LRG_OysterKnock_AnimBP
target_skeleton=&/M_LRG_OysterKnock_Skeleton
graphs [
 AnimGraph {nodes [
   n1 = Control Rig!(
    Source=Copy Pose From Mesh() {bUseAttachedParent=true}
    Alpha
    ParachuteOpen:bool=Get bIsParachuteOpened()
    isPlayingEmote:double=Get Emote() -> (Emote:double)
    WingsEquipped:double=Get OysterKnockGlider() -> (OysterKnockGlider:double)
    NOTfloating:double=Get IsNotFloating() -> (IsNotFloating:double)
    Floating:double=Get IsFloating() -> (IsFloating:double)
   ) -> (Pose) {
    alphaBoolBlend {blendOption=Linear}
    alphaScaleBiasClamp {bInterpResult=true, interpSpeedDecreasing=6, interpSpeedIncreasing=6}
    controlRigAssetReference {blueprintRigClass=&/M_LRG_OysterKnock_CtrlRig.M_LRG_OysterKnock_CtrlRig_C}
   }
   Output Pose(Result=Layered blend per bone(BasePose=@n1, BlendPoses_0=Slot 'DefaultSlot'(), BlendWeights_0=Get CustomIdleIsPlayingFloat() -> (CustomIdleIsPlayingFloat:float)) {layerSetup [{branchFilters [{boneName=sholderpad_r}, {boneName=sholderpad_l}]}]})
  ]}
 EventGraph {nodes [
   Event BlueprintUpdateAnimation()
   > n1 = Sequence!(execute) -> (then_0, then_1, then_2:Exec)
   > n2 = Cast To FortPlayerPawn(Object=GetOwningActor())
   > Is Valid(InputObject=Get OysterKnockSkelMesh() -> (OysterKnockSkelMesh:Skeletal Mesh Object Reference))
   > n3 = Set Current Glider(Current Glider:Athena Glider Item Definition Object Reference=GetItemDefinitionTypeFromLoadoutWithActor!(
     Actor:Actor Object Reference=@n2
     LoadoutSlot:Gameplay Tag Structure=(TagName="LoadoutSlot.Glider")
     CosmeticType:Athena Cosmetic Item Definition Class Reference=/Script/FortniteGame.AthenaGliderItemDefinition
    ) -> (OutCosmeticItemDef:Athena Glider Item Definition Object Reference)) -> (Output_Get:Athena Glider Item Definition Object Reference)
   n4 = Get GliderItemDef() -> (GliderItemDef:Athena Glider Item Definition Object Reference)
   > Set OysterKnockGlider(OysterKnockGlider:double="To Float (Boolean)"(InBool="Equal (Object)"(A=@n3, B=@n4) {is=title K2Node_CallFunction})) -> (Output_Get:double)
   n5 = Get bIsSkydiveFloating()
   Set IsNotFloating(execute=@n1.then_1, IsNotFloating:double="To Float (Boolean)"(InBool=NOT Boolean(A=@n5))) -> (Output_Get:double)
   > Set IsFloating(IsFloating:double="To Float (Boolean)"(InBool=Get bIsSkydiveFloating())) -> (Output_Get:double)
   > Set Emote(Emote:double="To Float (Boolean)"(InBool=Get bIsPlayingEmote())) -> (Output_Get:double)
   Branch(execute=@n1.then_2, Condition=Get bIsInFrontEnd())
   > Set CustomIdleIsPlayingFloat(CustomIdleIsPlayingFloat:float=GetCurveValue(self=GetOwnerFrontEndAnimInstance(), CurveName=Frontend_Idle_Bypass)) -> (Output_Get:float)
   TryGetPawnOwner()
  ]}
]
variables [
 Current Glider {type=AthenaGliderItemDefinition, default=null}
 CustomIdleIsPlayingFloat {type=number}
 Emote {type=number}
 GliderItemDef {type=AthenaGliderItemDefinition, default=/BRCosmetics/Athena/Items/Cosmetics/Gliders/Glider_OysterKnock}
 head rot {type=Rotator}
 IsFloating {type=number}
 IsNotFloating {type=number}
 OysterKnockGlider {type=number}
 OysterKnockSkelMesh {type=SkeletalMesh, default=&/M_LRG_OysterKnock}
 X percentage {type=number}
 Y percentage {type=number}
 Z percentage {type=number}
]
